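; Gameplay ability for the Very Rare tier of the PurpleStuff consumable in Athena.
; On activation it calls the parent implementation, runs generic ability setup and stores the player pawn from that setup.
; It then removes owner gameplay effects that grant the PurpleStuff item and cooldown tags, and ends immediately.
; Its end-ability event adds no logic.
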(event Ability|EventActivateAbility
  (|CallParentFunction)
  (bind (_event_data _fort_player_pawn) (Setup|SetupGenericAbility false false)
    (:Output
      (Variables|Default|SetPlayerPawn _fort_player_pawn)
      (Ability|RemoveGameplayEffectFromOwnerWithGrantedTags "(GameplayTags=((TagName=\"Athena.Item.PurpleStuff\")))")
      (Ability|RemoveGameplayEffectFromOwnerWithGrantedTags "(GameplayTags=((TagName=\"Athena.Cooldown.Item.PurpleStuff\")))")
      (Ability|EndAbility))))

(event Ability|EventOnEndAbility
  (params (bWasCancelled :type Boolean)))
